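; Static mesh actor Blueprint for the Academy Blundered Cape A prop that flags its mesh for the custom depth pass.
; The Construction Script enables Render Custom Depth on the actor's StaticMeshComponent, and BeginPlay applies the same setting again at runtime.
; No stencil value or other logic is set in these graphs.
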(event EventBeginPlay
  (Rendering|SetRenderCustomDepth (Variables|StaticMeshActor|GetStaticMeshComponent) true))

(fn ConstructionScript
  (Rendering|SetRenderCustomDepth (Variables|StaticMeshActor|GetStaticMeshComponent) true))
