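# Translucent character cape material for MistRaven Blundered that turns the cloth into animated, unlit smoke.
# A time-driven curve atlas row and local-position height gradient form the transition masks, while distorted panning noise supplies two-color smoke and vertex displacement.
# The smoke blends over the CharacterParent PreFX attributes, then passes through translucency sorting and PostFX.
# A randomized AnimationTime chain is left unconnected.
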Material /Hermes/Props/Academy/Academy_Blundered_Cape_A/Materials/M_MistRaven_Blundered_Cape
graph {
 properties {
  allowTranslucentCustomDepthWrites=true
  bAutomaticallySetUsageInEditor=false
  bDisableDepthTest=true
  bIsTranslucencyVelocityFromDepth=true
  bUseMaterialAttributes=true
  bUsedWithNanite=true
  bUsedWithStaticLighting=true
  blendMode=BLEND_Translucent
  translucencyLightingMode=TLM_SurfacePerPixelLighting
  usedShadingModels=Default Lit
 }
 nodes [
  Add(B=NamedRerouteUsage())
  n1 = Time()
  n2 = Frac(None=Multiply(A=Floor(None=@n1), B=Constant() {r=123.456}))
  n3 = Add(A=Constant() {r=159.4785}) {desc=seed}
  n4 = Add(A=@n2, B=DotProduct(A=@n2, B=Add(A=@n2, B=@n3) {constB=159.846}))
  n5 = ScalarParameter() {defaultValue=1, parameterName=AnimationTime}
  n6 = Add(A=LinearInterpolate(A=Constant() {r=2}, B=Constant() {r=6}, Alpha=Frac(None=@n4) {desc=H11}), B=@n5)
  n7 = Frac(None=Divide(A=Time(), B=@n6))
  Clamp(None=OneMinus(None=Multiply(A=@n7, B=@n6) {constB=0.5}), Max=@n5)
  n8 = MaterialFunctionCall() {materialFunction=&MF_CharacterParent_2023_PreFX}
  n9 = Constant()
  n10 = Subtract(A=Add(A=NamedRerouteUsage(), B=Constant() {r=0.8}), B=Multiply(A=NamedRerouteUsage()) {constB=0.9})
  n11 = SetMaterialAttributes!(
   MaterialAttributes
   Base Color=NamedRerouteUsage()
   Roughness=Constant() {r=1}
   Metallic=@n9
   Specular=@n9
   Shading Model=ShadingModel() {shadingModel=MSM_Unlit}
   Opacity=Multiply(A=Power(Base=@n10) {constExponent=0.5}, B=NamedRerouteUsage())
  ) {attributeSetTypes [
    #1
    #2
    #3
    #4
    #5
    #6
   ]}
  n12 = NamedRerouteUsage()
  n13 = GetMaterialAttributes(None=BlendMaterialAttributes(A=@n8.Result, B=@n11, Alpha=Saturate(None=@n12))) {attributeGetTypes [#6]}
  MaterialFunctionCall(
   Material=SetMaterialAttributes!(
    MaterialAttributes=@n13.MaterialAttributes
    World Position Offset=NamedRerouteUsage()
    Opacity=MaterialFunctionCall!(Opacity=@n13.Opacity) {materialFunction=/Game/Global/Materials/MaterialGym/Materials/SortedOutline/MF_SortTranslucency}.Result
   ) {attributeSetTypes [#7, #6]}
   M=@n8.M
   Pre Skinned Local Position=@n8.M
   Distance Blend=@n8.M
   Fresnel=@n8.M
  ) {materialFunction=&MF_CharacterParent_2023_PostFX}
  n14 = ComponentMask(None=NamedRerouteUsage()) {b=true}
  NamedRerouteDeclaration(None=Saturate(None=OneMinus(None=@n14))) {name=HeighGradient, nodeColor=<rgba 0.6406 1 0 1>}
  n15 = Multiply(A=TextureCoordinate() {coordinateIndex=1}) {constB=0.8}
  n16 = Panner(Coordinate=@n15) {speedX=0.1, speedY=-0.5}
  n17 = TextureSample(UVs=Panner(Coordinate=Multiply(A=@n15) {constB=1.333}) {speedX=0.05, speedY=-0.4}) {samplerSource=SSM_Wrap_WorldGroupSettings, texture=/Game/Effects/Fort_Effects/Emotes/UltraLight/Texture/T_SoftNoise_BC7}
  n18 = Multiply(A=Add(A=ComponentMask(None=@n17.RGB) {g=true, r=true}) {constB=-0.5}) {constB=0.3}
  n19 = Multiply(A=TextureSample(UVs=Add(A=@n16, B=@n18)) {samplerSource=SSM_Wrap_WorldGroupSettings, texture=&T_SoftNoise_BC7}.RGB) {constB=2}
  n20 = Multiply(A=Normalize(VectorInput=Add(A=@n19) {constB=-1}), B=Constant3Vector() {constant=<rgba 1 1 0.4 0>})
  n21 = Saturate(None=Multiply(A=NamedRerouteUsage()) {constB=-1})
  n22 = LinearInterpolate(Alpha=@n21) {constA=5, constB=15}
  n23 = LinearInterpolate(Alpha=@n21) {constB=0.5}
  n24 = NamedRerouteUsage()
  NamedRerouteDeclaration(None=Multiply(A=Multiply(A=@n20, B=@n22), B=Add(A=@n23, B=@n24))) {name=WPO, nodeColor=<rgba 1 0 0.9609 1>}
  n25 = Subtract(A=@n14, B=LinearInterpolate(Alpha=NamedRerouteUsage()) {constA=-0.5, constB=0.5})
  NamedRerouteDeclaration(None=Saturate(None=SmoothStep(Value=@n25) {constMax=0.4, constMin=0.6})) {name=BlendGradient, nodeColor=<rgba 0.5703 1 0 1>}
  n26 = Divide(A=LocalPosition() {localOrigin=Primitive}.XYZ) {constB=100}
  NamedRerouteDeclaration(None=Add(A=Add(A=@n26) {constB=0.5}, B=Constant3Vector() {constant=<rgba 0 0 0.5 0>})) {name=PSLP, nodeColor=<rgba 0 0.8594 1 1>}
  n27 = NamedRerouteUsage()
  n28 = Rotator(Coordinate=TextureCoordinate() {coordinateIndex=1}, Time=Constant() {r=1}) {speed=0.5}
  n29 = Subtract(A=Add(A=Multiply(A=@n27) {constB=0.25}, B=ComponentMask(None=@n28) {g=true}), B=Divide(A=Time()) {constB=1.2})
  n30 = CurveAtlasRowParameter(CurveTime=Frac(None=Divide(A=@n29) {constB=50})) {
   atlas=&/CA_MistRaven_Blundered_Cape
   curve=&/C_TransitionAnim
   parameterName=Curve
  }
  NamedRerouteDeclaration(None=@n30.R) {name=TimeLine, nodeColor=<rgba 0 1 0.8125 1>}
  n31 = Multiply(A=TextureCoordinate() {coordinateIndex=1}) {constB=3}
  n32 = Panner(Coordinate=@n31) {speedX=-0.15, speedY=-0.5}
  n33 = TextureSample(UVs=Panner(Coordinate=Multiply(A=@n31) {constB=1.333}) {speedX=0.1, speedY=-0.2}) {samplerSource=SSM_Wrap_WorldGroupSettings, texture=&T_SoftNoise_BC7}
  n34 = Multiply(A=Add(A=ComponentMask(None=@n33.RGB) {g=true, r=true}) {constB=-0.5}) {constB=0.2}
  NamedRerouteDeclaration(None=TextureSample(UVs=Add(A=@n32, B=@n34) {constB=0}) {samplerSource=SSM_Wrap_WorldGroupSettings, texture=&T_SoftNoise_BC7}.G) {name=SmokeDistorted, nodeColor=<rgba 0 0.3203 1 1>}
  n35 = Add(A=Power(Base=NamedRerouteUsage()) {constExponent=3}, B=Multiply(A=NamedRerouteUsage()) {constB=0.3})
  NamedRerouteDeclaration(None=LinearInterpolate(
    A=VectorParameter() {parameterName=Smoke_TopColor}.RGB
    B=VectorParameter() {parameterName=Smoke_BottomColor, defaultValue=<rgba 0.1615 0.0016 0 1>}.RGB
    Alpha=Saturate(None=@n35)
   )) {name=BaseColor, nodeColor=<rgba 1 0 0.0612 1>}
  n36 = TextureSampleParameter2D() {
   parameterName=FX2
   samplerType=SAMPLERTYPE_Masks
   texture=/Hermes/Props/Academy/Academy_Blundered_Cape_A/Textures/T_WraithForm_MistRaven_Cloak_FX2
  }
  NamedRerouteDeclaration(None=ComponentMask(None=@n36.R) {r=true}) {name=Alpha, nodeColor=<rgba 1 0.8203 0 1>}
 ]
}
